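# Physics asset for the TiltedParrot backpack cosmetic, setting up secondary motion for its left and right feathers.
# Each feather, lettered b through h on both sides, has a small base capsule followed by two larger capsules on dynamic bones.
# Constraints chain each base to its first and then second dynamic segment, with very tight swing and twist limits that loosen slightly toward the tip for stiff, subtle flex.
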PhysicsAsset /BRCosmetics/Accessories/FORT_Backpacks/Backpack_M_MED_TiltedParrot/Meshes/M_MED_TiltedParrot_Pack_Physics
bodies {
 base_feather_b_01_l [{shapeType=Capsule, length=0.505, radius=0.505, shapeName=base_feather_b_01_l_capsule, rotation=<rot 0 -90 0>}]
 base_feather_b_01_r [{shapeType=Capsule, length=0.505, radius=0.505, shapeName=base_feather_b_01_r_capsule, rotation=<rot 0 -90 0>}]
 base_feather_c_01_l [{shapeType=Capsule, length=0.505, radius=0.505, shapeName=base_feather_c_01_l_capsule, rotation=<rot 0 -90 0>}]
 base_feather_c_01_r [{shapeType=Capsule, length=0.505, radius=0.505, shapeName=base_feather_c_01_r_capsule, rotation=<rot 0 -90 0>}]
 base_feather_d_01_l [{shapeType=Capsule, length=0.505, radius=0.505, shapeName=base_feather_d_01_l_capsule, rotation=<rot 0 -90 0>}]
 base_feather_d_01_r [{shapeType=Capsule, length=0.505, radius=0.505, shapeName=base_feather_d_01_r_capsule, rotation=<rot 0 -90 0>}]
 base_feather_e_01_l [{shapeType=Capsule, length=0.505, radius=0.505, shapeName=base_feather_e_01_l_capsule, rotation=<rot 0 -90 0>}]
 base_feather_e_01_r [{shapeType=Capsule, length=0.505, radius=0.505, shapeName=base_feather_e_01_r_capsule, rotation=<rot 0 -90 0>}]
 base_feather_f_01_l [{shapeType=Capsule, length=0.505, radius=0.505, shapeName=base_feather_f_01_l_capsule, rotation=<rot 0 -90 0>}]
 base_feather_f_01_r [{shapeType=Capsule, length=0.505, radius=0.505, shapeName=base_feather_f_01_r_capsule, rotation=<rot 0 -90 0>}]
 base_feather_g_01_l [{shapeType=Capsule, length=0.505, radius=0.505, shapeName=base_feather_g_01_l_capsule, rotation=<rot 0 -90 0>}]
 base_feather_g_01_r [{shapeType=Capsule, length=0.505, radius=0.505, shapeName=base_feather_g_01_r_capsule, rotation=<rot 0 -90 0>}]
 base_feather_h_01_l [{shapeType=Capsule, length=0.505, radius=0.505, shapeName=base_feather_h_01_l_capsule, rotation=<rot 0 -90 0>}]
 base_feather_h_01_r [{shapeType=Capsule, length=0.505, radius=0.505, shapeName=base_feather_h_01_r_capsule, rotation=<rot 0 -90 0>}]
 dyn_feather_b_01_l [{shapeType=Capsule, length=1, radius=1, shapeName=dyn_feather_b_01_l_capsule, center=<-1 0 0>, rotation=<rot 90 0 0>}]
 dyn_feather_b_01_r [{shapeType=Capsule, length=1, radius=1, shapeName=dyn_feather_b_01_r_capsule, center=<1 0 0>, rotation=<rot 90 0 0>}]
 dyn_feather_b_02_l [{shapeType=Capsule, length=1, radius=1, shapeName=dyn_feather_b_02_l_capsule, center=<-1 0 0>, rotation=<rot 90 0 0>}]
 dyn_feather_b_02_r [{shapeType=Capsule, length=1, radius=1, shapeName=dyn_feather_b_02_r_capsule, center=<1 0 0>, rotation=<rot 90 0 0>}]
 dyn_feather_c_01_l [{shapeType=Capsule, length=1, radius=1, shapeName=dyn_feather_c_01_l_capsule, center=<-1 0 0>, rotation=<rot 90 0 0>}]
 dyn_feather_c_01_r [{shapeType=Capsule, length=1, radius=1, shapeName=dyn_feather_c_01_r_capsule, center=<1 0 0>, rotation=<rot 90 0 0>}]
 dyn_feather_c_02_l [{shapeType=Capsule, length=1, radius=1, shapeName=dyn_feather_c_02_l_capsule, center=<-1 0 0>, rotation=<rot 90 0 0>}]
 dyn_feather_c_02_r [{shapeType=Capsule, length=1, radius=1, shapeName=dyn_feather_c_02_r_capsule, center=<1 0 0>, rotation=<rot 90 0 0>}]
 dyn_feather_d_01_l [{shapeType=Capsule, length=1, radius=1, shapeName=dyn_feather_d_01_l_capsule, center=<-1 0 0>, rotation=<rot 90 0 0>}]
 dyn_feather_d_01_r [{shapeType=Capsule, length=1, radius=1, shapeName=dyn_feather_d_01_r_capsule, center=<1 0 0>, rotation=<rot 90 0 0>}]
 dyn_feather_d_02_l [{shapeType=Capsule, length=1, radius=1, shapeName=dyn_feather_d_02_l_capsule, center=<-1 0 0>, rotation=<rot 90 0 0>}]
 dyn_feather_d_02_r [{shapeType=Capsule, length=1, radius=1, shapeName=dyn_feather_d_02_r_capsule, center=<1 0 0>, rotation=<rot 90 0 0>}]
 dyn_feather_e_01_l [{shapeType=Capsule, length=1, radius=1, shapeName=dyn_feather_e_01_l_capsule, center=<-1 0 0>, rotation=<rot 90 0 0>}]
 dyn_feather_e_01_r [{shapeType=Capsule, length=1, radius=1, shapeName=dyn_feather_e_01_r_capsule, center=<1 0 0>, rotation=<rot 90 0 0>}]
 dyn_feather_e_02_l [{shapeType=Capsule, length=1, radius=1, shapeName=dyn_feather_e_02_l_capsule, center=<-1 0 0>, rotation=<rot 90 0 0>}]
 dyn_feather_e_02_r [{shapeType=Capsule, length=1, radius=1, shapeName=dyn_feather_e_02_r_capsule, center=<1 0 0>, rotation=<rot 90 0 0>}]
 dyn_feather_f_01_l [{shapeType=Capsule, length=1, radius=1, shapeName=dyn_feather_f_01_l_capsule, center=<-1 0 0>, rotation=<rot 90 0 0>}]
 dyn_feather_f_01_r [{shapeType=Capsule, length=1, radius=1, shapeName=dyn_feather_f_01_r_capsule, center=<1 0 0>, rotation=<rot 90 0 0>}]
 dyn_feather_f_02_l [{shapeType=Capsule, length=1, radius=1, shapeName=dyn_feather_f_02_l_capsule, center=<-1 0 0>, rotation=<rot 90 0 0>}]
 dyn_feather_f_02_r [{shapeType=Capsule, length=1, radius=1, shapeName=dyn_feather_f_02_r_capsule, center=<1 0 0>, rotation=<rot 90 0 0>}]
 dyn_feather_g_01_l [{shapeType=Capsule, length=1, radius=1, shapeName=dyn_feather_g_01_l_capsule, center=<-1 0 0>, rotation=<rot 90 0 0>}]
 dyn_feather_g_01_r [{shapeType=Capsule, length=1, radius=1, shapeName=dyn_feather_g_01_r_capsule, center=<1 0 0>, rotation=<rot 90 0 0>}]
 dyn_feather_g_02_l [{shapeType=Capsule, length=1, radius=1, shapeName=dyn_feather_g_02_l_capsule, center=<-1 0 0>, rotation=<rot 90 0 0>}]
 dyn_feather_g_02_r [{shapeType=Capsule, length=1, radius=1, shapeName=dyn_feather_g_02_r_capsule, center=<1 0 0>, rotation=<rot 90 0 0>}]
 dyn_feather_h_01_l [{shapeType=Capsule, length=1, radius=1, shapeName=dyn_feather_h_01_l_capsule, center=<-1 0 0>, rotation=<rot 90 0 0>}]
 dyn_feather_h_01_r [{shapeType=Capsule, length=1, radius=1, shapeName=dyn_feather_h_01_r_capsule, center=<1 0 0>, rotation=<rot 90 0 0>}]
 dyn_feather_h_02_l [{shapeType=Capsule, length=1, radius=1, shapeName=dyn_feather_h_02_l_capsule, center=<-1 0 0>, rotation=<rot 90 0 0>}]
 dyn_feather_h_02_r [{shapeType=Capsule, length=1, radius=1, shapeName=dyn_feather_h_02_r_capsule, center=<1 0 0>, rotation=<rot 90 0 0>}]
}
constraints [
 {bone1Name=dyn_feather_b_01_l, bone2Name=base_feather_b_01_l, swing1LimitDegrees=2, swing2LimitDegrees=2, twistLimitDegrees=0.5}
 {bone1Name=dyn_feather_b_02_l, bone2Name=dyn_feather_b_01_l, swing1LimitDegrees=4, swing2LimitDegrees=4, twistLimitDegrees=1}
 {bone1Name=dyn_feather_c_01_l, bone2Name=base_feather_c_01_l, swing1LimitDegrees=2, swing2LimitDegrees=2, twistLimitDegrees=0.5}
 {bone1Name=dyn_feather_c_02_l, bone2Name=dyn_feather_c_01_l, swing1LimitDegrees=4, swing2LimitDegrees=4, twistLimitDegrees=1}
 {bone1Name=dyn_feather_d_01_l, bone2Name=base_feather_d_01_l, swing1LimitDegrees=2, swing2LimitDegrees=2, twistLimitDegrees=0.5}
 {bone1Name=dyn_feather_d_02_l, bone2Name=dyn_feather_d_01_l, swing1LimitDegrees=4, swing2LimitDegrees=4, twistLimitDegrees=1}
 {bone1Name=dyn_feather_e_01_l, bone2Name=base_feather_e_01_l, swing1LimitDegrees=2, swing2LimitDegrees=2, twistLimitDegrees=0.5}
 {bone1Name=dyn_feather_e_02_l, bone2Name=dyn_feather_e_01_l, swing1LimitDegrees=4, swing2LimitDegrees=4, twistLimitDegrees=1}
 {bone1Name=dyn_feather_f_01_l, bone2Name=base_feather_f_01_l, swing1LimitDegrees=2, swing2LimitDegrees=2, twistLimitDegrees=0.5}
 {bone1Name=dyn_feather_f_02_l, bone2Name=dyn_feather_f_01_l, swing1LimitDegrees=4, swing2LimitDegrees=4, twistLimitDegrees=1}
 {bone1Name=dyn_feather_g_01_l, bone2Name=base_feather_g_01_l, swing1LimitDegrees=2, swing2LimitDegrees=2, twistLimitDegrees=0.5}
 {bone1Name=dyn_feather_g_02_l, bone2Name=dyn_feather_g_01_l, swing1LimitDegrees=4, swing2LimitDegrees=4, twistLimitDegrees=1}
 {bone1Name=dyn_feather_h_01_l, bone2Name=base_feather_h_01_l, swing1LimitDegrees=2, swing2LimitDegrees=2, twistLimitDegrees=0.5}
 {bone1Name=dyn_feather_h_02_l, bone2Name=dyn_feather_h_01_l, swing1LimitDegrees=4, swing2LimitDegrees=4, twistLimitDegrees=1}
 {bone1Name=dyn_feather_b_01_r, bone2Name=base_feather_b_01_r, swing1LimitDegrees=2, swing2LimitDegrees=2, twistLimitDegrees=0.5}
 {bone1Name=dyn_feather_b_02_r, bone2Name=dyn_feather_b_01_r, swing1LimitDegrees=4, swing2LimitDegrees=4, twistLimitDegrees=1}
 {bone1Name=dyn_feather_c_01_r, bone2Name=base_feather_c_01_r, swing1LimitDegrees=2, swing2LimitDegrees=2, twistLimitDegrees=0.5}
 {bone1Name=dyn_feather_c_02_r, bone2Name=dyn_feather_c_01_r, swing1LimitDegrees=4, swing2LimitDegrees=4, twistLimitDegrees=1}
 {bone1Name=dyn_feather_d_01_r, bone2Name=base_feather_d_01_r, swing1LimitDegrees=2, swing2LimitDegrees=2, twistLimitDegrees=0.5}
 {bone1Name=dyn_feather_d_02_r, bone2Name=dyn_feather_d_01_r, swing1LimitDegrees=4, swing2LimitDegrees=4, twistLimitDegrees=1}
 {bone1Name=dyn_feather_e_01_r, bone2Name=base_feather_e_01_r, swing1LimitDegrees=2, swing2LimitDegrees=2, twistLimitDegrees=0.5}
 {bone1Name=dyn_feather_e_02_r, bone2Name=dyn_feather_e_01_r, swing1LimitDegrees=4, swing2LimitDegrees=4, twistLimitDegrees=1}
 {bone1Name=dyn_feather_f_01_r, bone2Name=base_feather_f_01_r, swing1LimitDegrees=2, swing2LimitDegrees=2, twistLimitDegrees=0.5}
 {bone1Name=dyn_feather_f_02_r, bone2Name=dyn_feather_f_01_r, swing1LimitDegrees=4, swing2LimitDegrees=4, twistLimitDegrees=1}
 {bone1Name=dyn_feather_g_01_r, bone2Name=base_feather_g_01_r, swing1LimitDegrees=2, swing2LimitDegrees=2, twistLimitDegrees=0.5}
 {bone1Name=dyn_feather_g_02_r, bone2Name=dyn_feather_g_01_r, swing1LimitDegrees=4, swing2LimitDegrees=4, twistLimitDegrees=1}
 {bone1Name=dyn_feather_h_01_r, bone2Name=base_feather_h_01_r, swing1LimitDegrees=2, swing2LimitDegrees=2, twistLimitDegrees=0.5}
 {bone1Name=dyn_feather_h_02_r, bone2Name=dyn_feather_h_01_r, swing1LimitDegrees=4, swing2LimitDegrees=4, twistLimitDegrees=1}
]
