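# Animation Blueprint for the TiltedParrot backpack, posing its wings by player movement state and adding physics motion.
# Bool-blended additive bone rotations adjust wings for sprint, tactical sprint, skydive dive and parachute, chained through cached poses.
# Both swim-sprint nodes target the right wing with opposite yaws.
# A root-space RigidBody node follows, using gravity updated each frame by a deprecated override function, then the pack bone shifts in the front end.
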AnimBlueprint /BRCosmetics/Accessories/FORT_Backpacks/Backpack_M_MED_TiltedParrot/Meshes/M_MED_TiltedParrot_Pack_AnimBP
target_skeleton=&/M_MED_TiltedParrot_Pack_Skeleton
graphs [
 AnimGraph {nodes [
   n1 = RigidBody!(
    ComponentPose=Local To Component(LocalPose=Use cached pose 'Swim Sprinting_Cache'())
    Alpha
    OverrideWorldGravity:Vector=Get GravityOverride()
    ExternalForce
   ) -> (Pose) {
    bClampLinearTranslationLimitToRefPose=true
    bOverrideWorldGravity=true
    overridePhysicsAsset=&/M_MED_TiltedParrot_Pack_Physics
    simulationSpace=BaseBoneSpace
    alphaBoolBlend {blendOption=Linear}
    alphaScaleBiasClamp {bClampResult=true, bInterpResult=true, clampMax=0.8}
    baseBoneRef {boneName=root}
    componentAppliedLinearAccClamp=<4000 4000 4000>
    componentLinearAccScale=<0.4 0.4 0.4>
    componentLinearVelScale=<0.4 0.4 0.4>
    overrideWorldGravity {z=-980}
    simSpaceSettings {
     maxAngularAcceleration=1000
     maxAngularVelocity=3
     maxLinearAcceleration=700
     maxLinearVelocity=500
     velocityScaleZ=0
     worldAlpha=0.65
     externalLinearDragV=<0.2 0.2>
    }
   }
   n2 = "Transform (Modify) Bone - Bone: pack"!(ComponentPose=@n1, bAlphaBoolEnabled:bool=Get bIsInFrontEnd(), Translation="0,0.85,0", Rotation="0,0,0", Scale) -> (Pose) {
    alphaInputType=Bool
    rotationMode=BMM_Additive
    translationMode=BMM_Additive
    alphaBoolBlend {blendInTime=0.5, blendOption=Linear, blendOutTime=0.5}
    boneToModify {boneName=pack}
    translation {y=0.85}
   }
   Output Pose(Result=Component To Local(ComponentPose=@n2)) {layerGroup=None}
   Save cached pose 'InputPose_Cache'(Pose=Input Pose()) {cacheName=InputPose_Cache}
   n3 = Get bIsSprinting()
   n4 = "Transform (Modify) Bone - Bone: base_wing_r"!(
    ComponentPose=Local To Component(LocalPose=Use cached pose 'InputPose_Cache'())
    bAlphaBoolEnabled:bool=@n3
    Translation="0,0,0"
    Rotation="0,-5,0"
    Scale="1,1,1"
   ) -> (Pose) {
    alphaInputType=Bool
    rotationMode=BMM_Additive
    alphaBoolBlend {blendInTime=0.5, blendOption=Linear, blendOutTime=0.5}
    boneToModify {boneName=base_wing_r}
    rotation {yaw=-5}
   }
   n5 = "Transform (Modify) Bone - Bone: base_wing_l"!(ComponentPose=@n4, bAlphaBoolEnabled:bool=@n3, Translation="0,0,0", Rotation="0,-5,0", Scale="1,1,1") -> (Pose) {
    alphaInputType=Bool
    rotationMode=BMM_Additive
    alphaBoolBlend {blendInTime=0.5, blendOption=Linear, blendOutTime=0.5}
    boneToModify {boneName=base_wing_l}
    rotation {yaw=-5}
   }
   n6 = Get bIsTacticalSprinting()
   n7 = "Transform (Modify) Bone - Bone: base_wing_r"(ComponentPose=@n5, bAlphaBoolEnabled=@n6, Translation="0,0,0", Rotation="0,-20,0", Scale="1,1,1") {
    alphaInputType=Bool
    rotationMode=BMM_Additive
    alphaBoolBlend {blendInTime=0.5, blendOption=Linear, blendOutTime=0.5}
    boneToModify {boneName=base_wing_r}
    rotation {yaw=-20}
   }
   n8 = "Transform (Modify) Bone - Bone: base_wing_l"(ComponentPose=@n7, bAlphaBoolEnabled=@n6, Translation="0,0,0", Rotation="0,20,0", Scale="1,1,1") {
    alphaInputType=Bool
    rotationMode=BMM_Additive
    alphaBoolBlend {blendInTime=0.5, blendOption=Linear, blendOutTime=0.5}
    boneToModify {boneName=base_wing_l}
    rotation {yaw=20}
   }
   Save cached pose 'Sprinting_Cache'(Pose=Component To Local(ComponentPose=@n8)) {cacheName=Sprinting_Cache}
   n9 = Get bIsSkydiveDiveMode()
   n10 = "Transform (Modify) Bone - Bone: base_wing_r"!(ComponentPose=Local To Component(LocalPose=Use cached pose 'Sprinting_Cache'()), bAlphaBoolEnabled=@n9, Rotation="0,-35,0") -> (Pose) {
    alphaInputType=Bool
    rotationMode=BMM_Additive
    alphaBoolBlend {blendInTime=0.5, blendOption=Linear}
    alphaScaleBiasClamp {clampMax=0.8, interpSpeedIncreasing=15}
    boneToModify {boneName=base_wing_r}
    rotation {yaw=-35}
   }
   n11 = "Transform (Modify) Bone - Bone: base_wing_l"!(ComponentPose=@n10, bAlphaBoolEnabled=@n9, Rotation="0,35,0") -> (Pose) {
    alphaInputType=Bool
    rotationMode=BMM_Additive
    alphaBoolBlend {blendInTime=0.5, blendOption=Linear}
    alphaScaleBiasClamp {clampMax=0.8, interpSpeedIncreasing=15}
    boneToModify {boneName=base_wing_l}
    rotation {yaw=35}
   }
   n12 = Get bIsParachuteOpened()
   n13 = "Transform (Modify) Bone - Bone: base_wing_r"(ComponentPose=@n11, bAlphaBoolEnabled=@n12, Rotation="0,-40,0") {
    alphaInputType=Bool
    rotationMode=BMM_Additive
    alphaBoolBlend {blendInTime=1.5, blendOption=Linear, blendOutTime=1}
    alphaScaleBiasClamp {clampMax=0.8, interpSpeedIncreasing=15}
    boneToModify {boneName=base_wing_r}
    rotation {yaw=-40}
   }
   n14 = "Transform (Modify) Bone - Bone: base_wing_l"(ComponentPose=@n13, bAlphaBoolEnabled=@n12, Rotation="0,40,0") {
    alphaInputType=Bool
    rotationMode=BMM_Additive
    alphaBoolBlend {blendInTime=1.5, blendOption=Linear, blendOutTime=1}
    alphaScaleBiasClamp {clampMax=0.8, interpSpeedIncreasing=15}
    boneToModify {boneName=base_wing_l}
    rotation {yaw=40}
   }
   Save cached pose 'Skydive Mode_ParachuteOpen_Cache'(Pose=Component To Local(ComponentPose=@n14)) {cacheName=Skydive Mode_ParachuteOpen_Cache}
   n15 = Get bIsSwimSprinting() -> (bIsSwimSprinting:bool)
   n16 = "Transform (Modify) Bone - Bone: base_wing_r"(ComponentPose=Local To Component(LocalPose=Use cached pose 'Skydive Mode_ParachuteOpen_Cache'()), bAlphaBoolEnabled=@n15, Rotation="0,-20,0") {
    alphaInputType=Bool
    rotationMode=BMM_Additive
    alphaBoolBlend {blendInTime=1, blendOption=Linear, blendOutTime=0.1}
    boneToModify {boneName=base_wing_r}
    rotation {yaw=-20}
   }
   n17 = "Transform (Modify) Bone - Bone: base_wing_r"(ComponentPose=@n16, bAlphaBoolEnabled=@n15, Rotation="0,20,0") {
    alphaInputType=Bool
    rotationMode=BMM_Additive
    alphaBoolBlend {blendInTime=1, blendOption=Linear, blendOutTime=0.1}
    boneToModify {boneName=base_wing_r}
    rotation {yaw=20}
   }
   Save cached pose 'Swim Sprinting_Cache'(Pose=Component To Local(ComponentPose=@n17)) {cacheName=Swim Sprinting_Cache}
  ]}
 EventGraph {nodes [
   Event BlueprintUpdateAnimation()
   > n1 = Gravity Override State Function DEPRECATED(CharacterPartAnimInstance=Self-Reference(), GameTimeInSeconds=GetGameTimeInSeconds(), GravityOverrideParams=Get Gravity Override Params())
   > Set GravityOverride(GravityOverride=@n1)
   TryGetPawnOwner()
  ]}
]
variables [
 Gravity Override Params {type=GravityOverrideParamsStruct}
 GravityOverride {type=Vector}
 Jump Pose Alpha {type=number, default=1}
]
